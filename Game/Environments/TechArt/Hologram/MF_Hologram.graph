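# Material function that optionally turns incoming material attributes into a projected hologram, with capture-state colours driven by MPC_Hologram.
# A projection source and target define cone UVs and axis masks, alongside panning noise scanlines, fresnel and face lighting.
# It overrides base colour, adds emissive, dithers opacity mask with optional bottom fade, and distorts vertices through world position offset.
# An optional neck z-fight push adjusts that offset; several disconnected experimental branches remain.
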MaterialFunction /Game/Environments/TechArt/Hologram/MF_Hologram
graph {nodes [
  n1 = Power(Base=NamedRerouteUsage(), Exp=ScalarParameter() {defaultValue=3, parameterName=ScanLineContrast})
  n2 = DotProduct(A=Normalize(VectorInput=NamedRerouteUsage()), B=CameraVectorWS())
  n3 = Power(Base=@n2, Exp=ScalarParameter() {defaultValue=6, parameterName=Fresnel_Scanlines})
  n4 = Multiply(A=Time(), B=ScalarParameter() {defaultValue=1, parameterName=SpeedScanlineRect})
  n5 = Subtract(A=NamedRerouteUsage(), B=WorldPosition().XYZ)
  n6 = NamedRerouteUsage()
  n7 = CrossProduct(A=Constant3Vector() {constant=<rgba 0 0 1 0>}, B=@n6)
  n8 = Abs(None=DotProduct(A=@n5, B=CrossProduct(A=@n7, B=@n6)))
  n9 = Abs(None=DotProduct(A=@n5, B=@n7))
  n10 = Max(A=@n8, B=@n9)
  n11 = WorldPosition()
  n12 = ScalarParameter() {defaultValue=5000, parameterName=ConeLength}
  n13 = LinearInterpolate(B=ScalarParameter() {defaultValue=44.4457, parameterName=ProjectionSize_Z}, Alpha=Divide(A=DotProduct(A=@n6, B=@n11.XYZ), B=@n12))
  n14 = Sine(None=Add(A=Frac(None=@n4), B=Divide(A=@n10, B=@n13)))
  n15 = Saturate(None=@n14)
  n16 = ScalarParameter() {defaultValue=0.5, parameterName=ScaneLineRect_Brightness}
  Add(A=@n1, B=Multiply(A=Saturate(None=@n3), B=Multiply(A=@n15, B=@n16))) {desc=Scanlines Square removed}
  Add(A=@n8, B=@n9)
  n17 = ComponentMask(None=NamedRerouteUsage()) {r=true}
  n18 = ScalarParameter() {defaultValue=5, parameterName=Scanline_Tiling}
  n19 = ComponentMask(None=Panner(Coordinate=Multiply(A=@n17, B=@n18), Speed=ScalarParameter() {defaultValue=0.5, parameterName=Scanline_Speed})) {r=true}
  n20 = Add(A=@n18) {constB=0.1}
  n21 = ComponentMask(None=Panner(Coordinate=Multiply(A=@n17, B=@n20), Speed=ScalarParameter() {defaultValue=-0.4, parameterName=Scanline_Speed2})) {r=true}
  Add(A=Sine(None=@n19), B=Sine(None=@n21))
  n22 = MaterialFunctionCall(World Position=NamedRerouteUsage()) {materialFunction=&TransformToClipSpace}
  n23 = Subtract(A=ScreenPosition().ViewportUV, B=@n22.Screen aligned 0-1 UV)
  n24 = Multiply(
   A=MaterialFunctionCall(Vector or UVs=@n23) {materialFunction=&VectorToRadialValue}.Radial Coordinates
   B=AppendVector(
    A=ScalarParameter() {defaultValue=100, parameterName=ColorNoise_UV_x}
    B=ScalarParameter() {defaultValue=0.0183, parameterName=ColorNoise_UV_y}
   )
  )
  n25 = ComponentMask(None=Panner(Coordinate=@n24, Speed=ScalarParameter() {defaultValue=0.7356, parameterName=Noise_Pan})) {r=true}
  n26 = TextureSample(UVs=@n25) {samplerType=SAMPLERTYPE_Masks, texture=/Game/ArtTools/GodRays/GodRays_M}
  n27 = TextureSample(UVs=@n21) {samplerType=SAMPLERTYPE_Masks, texture=&GodRays_M}
  Add(A=@n26.R, B=@n27.R)
  CameraVectorWS()
  CameraVectorWS()
  Ceil(None=@n14)
  Clamp(None=@n22.Screen aligned 0-1 UV) {minDefault=-1}
  ComponentMask() {g=true, r=true}
  Constant3Vector()
  n28 = Divide(A=Distance(A=CameraPositionWS(), B=NamedRerouteUsage()), B=ScalarParameter() {defaultValue=1200, parameterName=Test Scale})
  Divide(A=@n24, B=@n28)
  n29 = Subtract(A=WorldPosition().XYZ, B=NamedRerouteUsage())
  n30 = MaterialFunctionCall() {materialFunction=&CameraDirectionVector}
  n31 = DotProduct(A=@n29, B=@n30.Result)
  n32 = Normalize(VectorInput=Subtract(A=WorldPosition().XYZ, B=CameraPositionWS()))
  n33 = DotProduct(A=@n30.Result, B=@n32)
  n34 = Subtract(A=@n29, B=Multiply(A=Divide(A=@n31, B=@n33), B=@n32))
  Divide(
   A=ComponentMask(None=Transform(None=@n34) {transformSourceType=TRANSFORMSOURCE_World, transformType=TRANSFORM_View}) {g=true, r=true}
   B=ScalarParameter() {defaultValue=20}
  )
  n35 = Length(None=@n22.Screen aligned 0-1 UV) {is=class MaterialExpressionLength}
  Divide(A=@n23, B=@n35)
  n36 = NamedRerouteUsage()
  n37 = ScalarParameter() {defaultValue=1, parameterName=Distor_Tiling}
  Divide(
   A=@n36
   B=Multiply(A=@n37, B=MaterialFunctionCall() {materialFunction=&ObjectScale}.Scale X)
  )
  n38 = WorldPosition()
  n39 = Length(None=Subtract(A=@n38.XYZ, B=NamedRerouteUsage()))
  n40 = NamedRerouteUsage()
  n41 = SquareRoot(None=Subtract(A=Multiply(A=@n39, B=@n39), B=Multiply(A=@n40, B=@n40)))
  n42 = NamedRerouteUsage()
  n43 = Subtract(A=@n38.XYZ, B=Add(A=Multiply(A=@n41, B=@n42), B=NamedRerouteUsage()))
  n44 = NamedRerouteDeclaration(None=Multiply(B=Normalize(VectorInput=@n43)) {constA=-1}) {name=Direction From Axis, nodeColor=<rgba 1 0 0.7266 1>}
  DotProduct(A=@n44)
  n45 = FunctionInput() {inputType=FunctionInput_MaterialAttributes}
  n46 = GetMaterialAttributes(None=@n45) {attributeGetTypes [
    #1
    #2
    #3
    #4
    #5
   ]}
  n47 = Multiply(A=@n46.OpacityMask, B=NamedRerouteUsage())
  n48 = Multiply(A=@n47, B=NamedRerouteUsage())
  n49 = MaterialFunctionCall() {materialFunction=&ObjectScale}
  n50 = ScalarParameter() {defaultValue=-10, group=NeckZFightFix, parameterName=NeckZfightPush}
  n51 = LocalPosition()
  n52 = VectorParameter() {group=NeckZFightFix, parameterName=NeckZFIghtFix_Position, defaultValue=<rgba 0 -4.1746 139.0125 1>}
  n53 = Distance(A=@n51.XYZ, B=@n52.RGB)
  n54 = ScalarParameter() {defaultValue=10, group=NeckZFightFix, parameterName=NeckZfight_Offset}
  n55 = ScalarParameter() {defaultValue=1.7195, group=NeckZFightFix, parameterName=NeckZfight_distanceMask}
  n56 = Saturate(None=Divide(A=Subtract(A=@n53, B=@n54), B=@n55))
  n57 = Multiply(A=Multiply(A=VertexNormalWS(), B=Multiply(A=@n49.Scale X, B=@n50)), B=OneMinus(None=@n56))
  n58 = Add(A=@n46.WorldPositionOffset, B=Multiply(A=@n56, B=NamedRerouteUsage()))
  n59 = Add(A=@n57, B=@n58)
  n60 = Subtract(A=WorldPosition().XYZ, B=ObjectPositionWS())
  n61 = SetMaterialAttributes!(
   MaterialAttributes=@n46.MaterialAttributes
   BaseColor=NamedRerouteUsage()
   EmissiveColor=Add(A=@n46.EmissiveColor, B=NamedRerouteUsage())
   OpacityMask=Multiply(A=StaticSwitchParameter(True=@n48, False=@n47) {defaultValue=true, parameterName=UseHologram_BottomFade}, B=NamedRerouteUsage())
   WorldPositionOffset=If(
    A=NamedRerouteUsage()
    "A > B"=StaticSwitchParameter(True=@n59, False=@n58) {group=NeckZFightFix, parameterName=UseNeckZFightFix}
    "A == B"=@n60
    "A < B"=@n60
   )
  ) {attributeSetTypes [
    #1
    #2
    #3
    #5
   ]}
  FunctionOutput(None=StaticSwitchParameter(True=@n61, False=@n45) {parameterName=UseHologram})
  n62 = NamedRerouteUsage()
  n63 = Constant() {r=1}
  If(A=@n62, "A > B"=@n63, "A == B"=Constant(), "A < B"=@n63) {constB=1}
  n64 = NamedRerouteUsage()
  n65 = CameraPositionWS()
  n66 = NamedRerouteDeclaration(None=Normalize(VectorInput=Subtract(A=@n64, B=@n65))) {name=CameraVector, nodeColor=<rgba 0 0.7422 1 1>}
  n67 = Constant3Vector() {constant=<rgba 0 0 1 0>}
  n68 = Transform(None=Constant3Vector() {constant=<rgba 0 0 1 0>}) {transformSourceType=TRANSFORMSOURCE_Camera}
  n69 = ScalarParameter() {defaultValue=599.9095, parameterName=Holo_ UV_Scale}
  MaterialFunctionCall(
   Plane Normal Axis=NamedRerouteUsage()
   Plane X Axis=Normalize(VectorInput=CrossProduct(A=@n66, B=@n67))
   Plane Center=Add(A=Multiply(A=@n68) {constB=10000}, B=@n64)
   UV Scale=AppendVector(A=@n69, B=@n69)
  ) {materialFunction=&VirtualPlaneCoordinates}
  MaterialFunctionCall() {materialFunction=&VectorToRadialValue}
  Multiply(A=@n26.R, B=@n27.R)
  Multiply(A=@n24, B=@n28)
  n70 = NamedRerouteUsage()
  n71 = CrossProduct(A=@n70, B=Constant3Vector() {constant=<rgba 0 0 1 0>})
  n72 = DotProduct(A=@n44, B=Normalize(VectorInput=CrossProduct(A=@n70, B=@n71)))
  n73 = Sine(None=Cosine(None=Multiply(A=@n72) {constB=3.142}))
  n74 = ScalarParameter() {defaultValue=0.5, parameterName=CenterLines}
  Multiply(A=@n73, B=@n74)
  n75 = CollectionParameter() {collection=&/MPC_Hologram, parameterName=ColorSwitch}
  n76 = NamedRerouteUsage()
  Multiply(A=@n75, B=OneMinus(None=@n76))
  Multiply(A=@n23, B=@n35)
  n77 = CollectionParameter() {collection=&MPC_Hologram, parameterName=FadeOpacity}
  Multiply(A=@n77, B=OneMinus(None=@n62))
  Multiply(A=@n49.Scale X, B=@n55)
  Multiply(A=@n54, B=@n49.Scale X)
  n78 = Power(Base=DotProduct(A=NamedRerouteUsage(), B=CameraVectorWS()), Exp=ScalarParameter() {defaultValue=1.7997, parameterName=Holo_Fresnel_Mask_C})
  n79 = ScalarParameter() {defaultValue=2, parameterName=Holo_Fresnel_Mask_A}
  n80 = ScalarParameter() {defaultValue=-1.0044, parameterName=Holo_Fresnel_Mask_B}
  n81 = NamedRerouteUsage()
  Multiply(
   A=LinearInterpolate(
    A=ScalarParameter() {defaultValue=0.5, parameterName=Holo_Opacity_B}
    B=ScalarParameter() {defaultValue=1.1626, parameterName=Holo_Opacity_A}
    Alpha=Saturate(None=@n78)
   )
   B=Saturate(None=LinearInterpolate(A=@n79, B=@n80, Alpha=@n81))
  )
  NamedRerouteDeclaration(None=TransformPosition(None=VectorParameter() {parameterName=ProjectionSource, defaultValue=<rgba 100 0 -150 1>}.RGB) {transformType=TRANSFORMPOSSOURCE_World}) {name=ProjectionSourceWS, nodeColor=<rgba 1 0.3984 0 1>}
  NamedRerouteDeclaration(None=@n46.Normal) {name=NormalsOrg, nodeColor=<rgba 0.1719 1 0 1>}
  NamedRerouteDeclaration(None=@n77) {name=FadeOpacity, nodeColor=<rgba 1 0.7969 0 1>}
  n82 = ScalarParameter() {defaultValue=0.65, parameterName=Main_Opacity}
  n83 = Multiply(A=NamedRerouteUsage(), B=@n82)
  NamedRerouteDeclaration(None=MaterialFunctionCall(Alpha Threshold=StaticSwitchParameter(True=@n83, False=@n82) {parameterName=UseHologramSphereMask}) {materialFunction=&DitherTemporalAA}.Result) {name=HologramOpacityMasked Dithered, nodeColor=<rgba 1 0.6328 0 1>}
  n84 = TextureSample(UVs=Multiply(A=ScalarParameter() {defaultValue=0.5, parameterName=Holo_ColorNoiseGradient}, B=NamedRerouteUsage())) {samplerType=SAMPLERTYPE_LinearColor, texture=/Game/Effects/Fort_Effects/Gameplay/SmokeDustExplosions/Textures/T_NoisePack_LinearColor}
  n85 = Multiply(
   A=Desaturation(None=@n84.RGB, Fraction=ScalarParameter() {defaultValue=0.5, parameterName=Holo_DesatNoise})
   B=Fresnel(ExponentIn=ScalarParameter() {defaultValue=1, parameterName=Holo_Fresnel}, Normal=NamedRerouteUsage()) {baseReflectFraction=0}
  )
  n86 = NamedRerouteUsage()
  n87 = LinearInterpolate(
   A=VectorParameter() {parameterName=ScanLineColor, defaultValue=<rgba 0.6213 0.8011 1 1>}.RGB
   B=VectorParameter() {parameterName=ScanLineColor2, defaultValue=<rgba 1 0.8531 0 1>}.RGB
   Alpha=NamedRerouteUsage()
  )
  n88 = VectorParameter() {parameterName=Contested, defaultValue=<rgba 1 0 0 1>}
  n89 = NamedRerouteUsage()
  n90 = Multiply(A=Multiply(A=Multiply(A=@n85, B=@n86), B=LinearInterpolate(A=@n87, B=@n88.RGB, Alpha=@n89)), B=NamedRerouteUsage())
  n91 = Power(
   Base=Subtract(A=OneMinus(None=@n2), B=ScalarParameter() {defaultValue=0.25, parameterName=Fresnel_Scanlines_Offset})
   Exp=ScalarParameter() {defaultValue=2, parameterName=Fresnel_Scanlines}
  )
  n92 = Multiply(
   A=LinearInterpolate(
    A=ScalarParameter() {defaultValue=0.015, parameterName=ScanlineBrightness_inner}
    B=ScalarParameter() {defaultValue=0.3, parameterName=ScanlineBrightness_Edges}
    Alpha=Saturate(None=@n91)
   )
   B=@n1
  )
  n93 = VectorParameter() {parameterName=ScanLineColor2nd, defaultValue=<rgba 8.7718 31.9686 50 1>}
  n94 = VectorParameter() {parameterName=ScanLineColor2nd_2, defaultValue=<rgba 30 27.5689 0 1>}
  n95 = NamedRerouteUsage()
  n96 = NamedRerouteDeclaration(None=LinearInterpolate(
    A=LinearInterpolate(A=@n93.RGB, B=@n94.RGB, Alpha=@n95)
    B=VectorParameter() {parameterName=Contested2, defaultValue=<rgba 50 0 0 1>}.RGB
    Alpha=NamedRerouteUsage()
   )) {name=Colors, nodeColor=<rgba 0 1 0.9297 1>}
  n97 = NamedRerouteUsage()
  n98 = ScalarParameter() {defaultValue=0.5, parameterName=FaceMask_Min}
  n99 = Multiply(A=Multiply(A=Multiply(A=@n92, B=@n96), B=NamedRerouteUsage()), B=Saturate(None=Add(A=@n97, B=@n98)))
  NamedRerouteDeclaration(None=Add(A=Add(A=@n90, B=@n99), B=NamedRerouteUsage())) {name=Emissive Final, nodeColor=<rgba 0 1 0.6016 1>}
  NamedRerouteDeclaration(None=@n84.R) {name=Color Noise Red Channel, nodeColor=<rgba 0 1 0.0859 1>}
  n100 = ObjectPositionWS()
  n101 = MaterialFunctionCall() {materialFunction=&ObjectScale}
  n102 = Subtract(
   A=Subtract(A=WorldPosition().Z, B=ComponentMask(None=@n100) {b=true})
   B=Multiply(A=@n101.Scale X, B=ScalarParameter() {defaultValue=-70, parameterName=BottomFade_Offset})
  )
  n103 = Multiply(A=@n101.Scale X, B=ScalarParameter() {defaultValue=50, parameterName=BottomFade_Distance})
  NamedRerouteDeclaration(None=Saturate(None=Divide(A=@n102, B=@n103))) {name=BottomFade, nodeColor=<rgba 1 0 0.8672 1>}
  n104 = Distance(
   A=WorldPosition() {worldPositionShaderOffset=WPT_ExcludeAllShaderOffsets}.XYZ
   B=TransformPosition(None=VectorParameter() {parameterName=Hologram_Face_SphereMask_Location, defaultValue=<rgba 0 10 153 1>}.RGB) {transformType=TRANSFORMPOSSOURCE_World}
  )
  n105 = MaterialFunctionCall() {materialFunction=&ObjectScale}
  n106 = Multiply(A=@n105.Scale X, B=ScalarParameter() {defaultValue=-12, parameterName=Hologram_Face_SphereMask_Offset})
  n107 = ScalarParameter() {defaultValue=40, parameterName=Hologram_Face_SphereMask_FadeDistance}
  n108 = Saturate(None=Divide(A=Add(A=@n104, B=@n106), B=Multiply(A=@n105.Scale X, B=@n107)))
  NamedRerouteDeclaration(None=@n108) {name=FaceSphereMask, nodeColor=<rgba 1 0 0.5625 1>}
  NamedRerouteDeclaration(None=Saturate(None=Add(A=@n75, B=@n76))) {name=ColorSwitch, nodeColor=<rgba 0 0.2266 1 1>}
  n109 = NamedRerouteUsage()
  n110 = VectorParameter() {parameterName=FaceLightingDirection, defaultValue=<rgba 0.344 0.216 -0.1754 1>}
  n111 = DotProduct(A=Normalize(VectorInput=Transform(None=@n110.RGB) {transformSourceType=TRANSFORMSOURCE_Local}), B=NamedRerouteUsage())
  n112 = ScalarParameter() {defaultValue=2, parameterName=FaceLightingFalloff}
  n113 = Multiply(
   A=Multiply(A=Power(Base=@n111, Exp=@n112), B=NamedRerouteUsage())
   B=ScalarParameter() {defaultValue=0.01, parameterName=FaceLightBrightness}
  )
  n114 = Multiply(
   A=NamedRerouteDeclaration(None=@n46.BaseColor) {name=BaseColorOrg, nodeColor=<rgba 1 0.9844 0 1>}
   B=ScalarParameter() {defaultValue=0.178, parameterName=FaceLightBrightness2}
  )
  NamedRerouteDeclaration(None=Multiply(A=OneMinus(None=@n109), B=Add(A=@n113, B=@n114))) {name=FaceLighting, nodeColor=<rgba 1 0.6094 0 1>}
  n115 = Subtract(A=WorldPosition().XYZ, B=NamedRerouteUsage())
  n116 = VectorParameter() {parameterName=ProjectionTarget, defaultValue=<rgba 0 0 0 1>}
  n117 = Normalize(VectorInput=Subtract(A=TransformPosition(None=@n116.RGB) {transformType=TRANSFORMPOSSOURCE_World}, B=NamedRerouteUsage()))
  n118 = NamedRerouteDeclaration(None=@n117) {name=ConeDirection, nodeColor=<rgba 1 0.6563 0 1>}
  n119 = Length(None=CrossProduct(A=@n115, B=@n118))
  n120 = NamedRerouteDeclaration(None=@n119) {name=DistanceFromAxis, nodeColor=<rgba 0 1 0.9766 1>}
  n121 = Subtract(
   A=Divide(A=@n120, B=MaterialFunctionCall() {materialFunction=&ObjectScale}.Scale X)
   B=ScalarParameter() {defaultValue=10, parameterName=CenterMask_Offset}
  )
  n122 = ScalarParameter() {defaultValue=20, parameterName=CenterMask_Distance}
  NamedRerouteDeclaration(None=Saturate(None=Divide(A=@n121, B=@n122))) {name=CenterMask, nodeColor=<rgba 0.0547 0 1 1>}
  n123 = CollectionParameter() {collection=&MPC_Hologram, parameterName=Contested}
  n124 = CollectionParameter() {collection=&MPC_Hologram, parameterName=EnemyCapture}
  NamedRerouteDeclaration(None=Saturate(None=Add(A=@n123, B=@n124))) {name=Contested, nodeColor=<rgba 0.8281 1 0 1>}
  NamedRerouteDeclaration(None=CollectionParameter() {collection=&MPC_Hologram, parameterName=Captured}) {name=Captured, nodeColor=<rgba 0 0.2266 1 1>}
  n125 = VectorParameter() {parameterName=Hologram_SphereMask_Location, defaultValue=<rgba 0 0 0 1>}
  n126 = MaterialFunctionCall() {materialFunction=&ObjectScale}
  n127 = Add(
   A=Distance(A=WorldPosition().XYZ, B=TransformPosition(None=@n125.RGB) {transformType=TRANSFORMPOSSOURCE_World})
   B=Multiply(A=@n126.Scale X, B=ScalarParameter() {parameterName=Hologram_SphereMask_Offset})
  )
  n128 = Multiply(A=@n126.Scale X, B=ScalarParameter() {defaultValue=50, parameterName=Hologram_SphereMask_FadeDistance})
  NamedRerouteDeclaration(None=Saturate(None=Divide(A=@n127, B=@n128))) {name=SphereMask, nodeColor=<rgba 0.6641 1 0 1>}
  n129 = DotProduct(A=Divide(A=@n115, B=@n12), B=@n118)
  n130 = LinearInterpolate(A=ScalarParameter() {parameterName=Radius_Start}, B=ScalarParameter() {defaultValue=245.0628, parameterName=ConeWidth}, Alpha=@n129)
  NamedRerouteDeclaration(None=AppendVector(A=Divide(A=@n120, B=@n130), B=@n129) {desc=Cone UVS}) {name=ConeUVs, nodeColor=<rgba 0.9219 0 1 1>}
  NamedRerouteDeclaration(None=Add(
    A=TextureSample(UVs=@n19) {texture=/Game/Global/Textures/Default/Noise/T_godray_Noise}.R
    B=TextureSample(UVs=@n21) {texture=&T_godray_Noise}.R
   )) {name=Scanlines, nodeColor=<rgba 1 0 0.5156 1>}
  n131 = TextureSampleParameter2D() {parameterName=BaseColor, texture=&T_BlacksRGB}
  n132 = TextureCoordinate()
  n133 = LinearInterpolate(B=ScalarParameter() {defaultValue=0.005, parameterName=Chromatic}, Alpha=NamedRerouteUsage())
  n134 = TextureSampleParameter2D(UVs=Add(A=@n132, B=@n133)) {parameterName=BaseColor, texture=&T_BlacksRGB}
  n135 = Add(A=@n132, B=Multiply(A=@n133) {constB=2})
  NamedRerouteDeclaration(None=AppendVector(
    A=AppendVector(A=@n131.R, B=@n134.G)
    B=TextureSampleParameter2D(UVs=@n135) {parameterName=BaseColor, texture=&T_BlacksRGB}.B
   )) {name=BaseColor With Chromatic, nodeColor=<rgba 0 0.4609 1 1>}
  n136 = Normalize(VectorInput=Subtract(A=WorldPosition().XYZ, B=NamedRerouteUsage()))
  n137 = Multiply(
   A=MaterialFunctionCall() {materialFunction=&ObjectScale}.Scale X
   B=NamedRerouteUsage()
  )
  NamedRerouteDeclaration(None=Multiply(A=NamedRerouteUsage(), B=Multiply(A=@n136, B=@n137))) {name=WPODistortion, nodeColor=<rgba 0 1 0.1797 1>}
  n138 = Panner(Coordinate=Multiply(A=@n36, B=@n37), Speed=ScalarParameter() {defaultValue=-0.25, parameterName=Distort_Speed})
  n139 = LinearInterpolate(
   A=ScalarParameter() {defaultValue=-9.7, parameterName=WPODistort_min}
   B=ScalarParameter() {defaultValue=21.9468, parameterName=WPODistort_Max}
   Alpha=NamedRerouteDeclaration(None=TextureSample(UVs=@n138) {texture=&T_godray_Noise}.R) {name=ScanlineDistortionTex, nodeColor=<rgba 0.6875 0 1 1>}
  )
  NamedRerouteDeclaration(None=Clamp(None=@n139) {clampMode=CMODE_ClampMin}) {name=ScaneLineDistortionWithDistance, nodeColor=<rgba 0.8516 1 0 1>}
  NamedRerouteDeclaration(None=Saturate(None=@n139)) {name=ScaneLineDistortion0to1, nodeColor=<rgba 1 0.9141 0 1>}
  NamedRerouteUsage()
  Normalize(VectorInput=Subtract(A=ObjectPositionWS(), B=CameraPositionWS())) {desc=CameraVector}
  Normalize(VectorInput=Subtract(A=NamedRerouteUsage(), B=CameraPositionWS()))
  ObjectPositionWS()
  Power(Base=@n73, Exp=@n74)
  Reroute(None=@n119) {desc=DistanceFromAxis}
  Reroute(None=@n117)
  Saturate(None=Add(A=@n108, B=ScalarParameter() {defaultValue=0.5, parameterName=FaceMask_Min}))
  Saturate(None=Sine(None=Divide()))
  SetMaterialAttributes!(MaterialAttributes, EmissiveColor=@n56, WorldPositionOffset=@n57) {attributeSetTypes [#2, #5]}
  Sine(None=Saturate(None=Sine(None=@n25)))
  Subtract(A=@n51.XYZ, B=Multiply(A=@n52.RGB, B=@n49.Scale X))
  TextureSample(UVs=@n25) {samplerType=SAMPLERTYPE_Masks, texture=/Engine/Functions/Engine_MaterialFunctions02/ExampleContent/Textures/SineWaveHeightMap}
  TextureSample(UVs=@n19) {samplerType=SAMPLERTYPE_Masks, texture=&SineWaveHeightMap}
  WorldPosition()
 ]}
